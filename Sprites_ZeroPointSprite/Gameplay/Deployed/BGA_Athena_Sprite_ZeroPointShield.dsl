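; Deployed Zero Point Sprite shield zone that grants players an ability and refills dash charges.
; BeginPlay scales the collision sphere and area particles from a scalable-float radius.
; On overlap, it applies the grant effect unless tagged active, adds missing charges via SetByCaller, broadcasts a charge verb, and plays sound-library audio, looping only for the viewed pawn.
; Exit removes buff-tagged effects and the sound library; EndShield hides and expires the actor.
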(event Custom|SoundLibraryCheck
  (params (Target :type Actor%20Object%20Reference) (Add :type Boolean))
  (bind _returnvalue (Actor|GetComponentbyClass Target "/Script/SoundLibrary.SoundLibraryComponent"))
  (bind _spritesoundlib (Variables|Default|GetSpriteSoundLib))
  ; Adding and removing Libraries
  (Utilities|IsValid _returnvalue
    (:Is%20Valid
      (if Add
        (SoundLibrary|AddLibrary _returnvalue _spritesoundlib :!class /Script/SoundLibrary.SoundLibraryComponent)
        (else
          (SoundLibrary|RemoveLibrary _returnvalue _spritesoundlib :!class /Script/SoundLibrary.SoundLibraryComponent))))))

(event Custom|SoundLibTrigger
  (params (Play :type Boolean) (InEventName :type Gameplay%20Tag%20Structure) (Target :type Actor%20Object%20Reference))
  (bind _returnvalue (Actor|GetComponentbyClass Target "/Script/SoundLibrary.SoundLibraryComponent"))
  ; Sound library logic
  (if (Utilities|IsValid _returnvalue :!class /Script/Engine.KismetSystemLibrary)
    (if Play
      (SoundLibrary|PlayEvent _returnvalue InEventName)
      (else
        (SoundLibrary|StopEvent _returnvalue InEventName)))))

(event EventBeginPlay
  (bind _row_aoeradius (Variables|Default|GetRow_AoeRadius))
  (bind _returnvalue (ScalableFloat|GetValueatLevel _row_aoeradius))
  (Components|Sphere|SetSphereRadius (Variables|Default|GetZeroPointShield_Collision) _returnvalue true)
  (Effects|Components|ParticleSystem|SetFloatParameter (Variables|Default|GetNS_ZeroPoint_AreaOfEffect) "AoE_Size" (* _returnvalue 2.0)))

(event Custom|AddEffect
  (params (Actor :type Actor%20Object%20Reference))
  (bind _returnvalue (Ability|GetAbilitySystemComponent Actor))
  (bind (_bsuccessfullyfoundattribute _returnvalue_2) (Ability|Attribute|GetFloatAttribute Actor "(AttributeName=\"CurrentCharges\",Attribute=/Script/Sprites_ZeroPointSpriteRuntime.ZeroPointSpriteAttributeSet:CurrentCharges,AttributeOwner=\"/Script/CoreUObject.Class'/Script/Sprites_ZeroPointSpriteRuntime.ZeroPointSpriteAttributeSet'\")"))
  (bind _row_numcharges (Variables|Default|GetRow_NumCharges))
  (bind _returnvalue_3 (ScalableFloat|GetValueatLevel _row_numcharges))
  (bind _returnvalue_4 (- _returnvalue_3 _returnvalue_2))
  (Utilities|IsValid _returnvalue
    (:Is%20Valid
      (Utilities|FlowControl|Sequence
        (:then_0
          (if (not (GameplayTags|HasMatchingGameplayTag (Utilities|Casting|ToGameplayTagAssetInterface Actor) (Variables|Default|GetZeroPointActiveTag)))
            (bind _returnvalue_5 (Spec|MakeSpecHandle(ByClass) "/Sprites_ZeroPointSprite/Gameplay/Deployed/GE_Athena_Sprite_ZeroPointShield_GrantAbility.GE_Athena_Sprite_ZeroPointShield_GrantAbility_C" self self))
            (GameplayEffects|ApplyGameplayEffectSpecToSelf _returnvalue _returnvalue_5)))
        (:then_1
          (if (< _returnvalue_2 _returnvalue_3)
            (bind _returnvalue_6 (Ability|GameplayEffect|AssignTagSetbyCallerMagnitude (GameplayEffects|MakeOutgoingSpec _returnvalue "/Sprites_ZeroPointSprite/Gameplay/Deployed/GE_Athena_Sprites_ZeroPointShield_DashCharges.GE_Athena_Sprites_ZeroPointShield_DashCharges_C" 0.0 (GameplayEffects|MakeEffectContext _returnvalue)) "(TagName=\"SetByCaller.DashCharges\")" _returnvalue_4))
            (GameplayEffects|ApplyGameplayEffectSpecToSelf _returnvalue _returnvalue_6)
            (CustomVerb|BroadcastCustomVerbMessage (CustomVerb|MakeCustomVerbMessage _returnvalue "(TagName=\"CustomVerb.ZeroPointSprite.ChargeAdded\")" (Math|Float|Truncate _returnvalue_4)))))))))

(event OnComponentBeginOverlap(ZeroPointShield_Collision)
  (params (OverlappedComponent :type Primitive%20Component%20Object%20Reference) (OtherActor :type Actor%20Object%20Reference) (OtherComp :type Primitive%20Component%20Object%20Reference) (OtherBodyIndex :type Integer) (bFromSweep :type Boolean) (SweepResult :type Hit%20Result%20Structure%20%28by%20ref%29))
  (bind _asplayer_pawn_athena (Utilities|Casting|CastToPlayerPawn_Athena OtherActor))
  (Utilities|FlowControl|Sequence
    (:then_0
      ; Trigger enter and inside loop audio
      (CallFunction|SoundLibraryCheck _asplayer_pawn_athena true)
      (CallFunction|SoundLibTrigger true "(TagName=\"SoundLibrary.Sprite.Zone.Enter\")" _asplayer_pawn_athena)
      (if (Pawn|IsLocallyViewed _asplayer_pawn_athena)
        (CallFunction|SoundLibTrigger true "(TagName=\"SoundLibrary.Sprite.Zone.Loop\")" _asplayer_pawn_athena)))
    (:then_1
      (CallFunction|AddEffect _asplayer_pawn_athena))))

(event OnComponentEndOverlap(ZeroPointShield_Collision)
  (params (OverlappedComponent :type Primitive%20Component%20Object%20Reference) (OtherActor :type Actor%20Object%20Reference) (OtherComp :type Primitive%20Component%20Object%20Reference) (OtherBodyIndex :type Integer))
  (bind _asplayer_pawn_athena (Utilities|Casting|CastToPlayerPawn_Athena OtherActor))
  (GameplayEffects|RemoveActiveEffectswithTags (Ability|GetAbilitySystemComponent _asplayer_pawn_athena) "(GameplayTags=((TagName=\"Item.Sprite.ZeroPoint.BuffActive\")))")
  ; Trigger exiting and stop loop audio
  (CallFunction|SoundLibTrigger true "(TagName=\"SoundLibrary.Sprite.Zone.Exit\")" _asplayer_pawn_athena)
  (CallFunction|SoundLibTrigger false "(TagName=\"SoundLibrary.Sprite.Zone.Loop\")" _asplayer_pawn_athena)
  (CallFunction|SoundLibraryCheck _asplayer_pawn_athena))

(event Custom|EndShield
  (Rendering|SetActorHiddenInGame true)
  (Actor|SetLifeSpan 0.01))

(fn ConstructionScript)
