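; Granted gameplay ability for the Zero Point sprite shield that manages its dash charges.
; On activation it applies the active effect to its owner, then waits for the active dashing tag to be added.
; When the tag is added it applies a charge-removal effect and, if CurrentCharges reads zero, waits briefly before removing effects tagged as granting Zero Point dashes.
; Ending the ability removes the active effect by its stored handle and stops the tag listener.
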(event Ability|EventActivateAbility
  (bind (_bsuccessfullyfoundattribute _returnvalue) (Ability|Attribute|GetFloatAttribute (CallFunction|GetAvatarasFortPlayerPawn) "(AttributeName=\"CurrentCharges\",Attribute=/Script/Sprites_ZeroPointSpriteRuntime.ZeroPointSpriteAttributeSet:CurrentCharges,AttributeOwner=\"/Script/CoreUObject.Class'/Script/Sprites_ZeroPointSpriteRuntime.ZeroPointSpriteAttributeSet'\")"))
  (bind _returnvalue_2 (Ability|ApplyGameplayEffectToOwner "/Sprites_ZeroPointSprite/Gameplay/Deployed/GE_Athena_Sprite_ZeroPointShield_ActiveEffect.GE_Athena_Sprite_ZeroPointShield_ActiveEffect_C"))
  (Variables|Default|SetZeroPointDashGE _returnvalue_2)
  (bind _asynctaskproxy (Ability|Tasks|WaitGameplayTagAdd (Variables|Default|GetActiveDashingTag))
    (:then
      (Variables|Default|SetAsync_TagAdded _asynctaskproxy))
    (:Added
      (Ability|ApplyGameplayEffectToOwner "/Sprites_ZeroPointSprite/Gameplay/Deployed/GE_Athena_Sprites_ZeroPointShield_RemoveCharge.GE_Athena_Sprites_ZeroPointShield_RemoveCharge_C")
      (if (== _returnvalue 0)
        (Utilities|FlowControl|Delay 0.2)
        (Ability|RemoveGameplayEffectFromOwnerWithAssetTags "(GameplayTags=((TagName=\"Granted.Abilities.ZeroPoint.Dashes\")))")))))

(event Ability|EventOnEndAbility
  (params (bWasCancelled :type Boolean))
  (Ability|RemoveGameplayEffectFromOwnerWithHandle (Variables|Default|GetZeroPointDashGE))
  (GameplayTasks|EndTask (Variables|Default|GetAsync_TagAdded)))
