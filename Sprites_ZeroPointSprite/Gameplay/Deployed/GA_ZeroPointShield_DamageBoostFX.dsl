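; Passive ability, apparently for the Zero Point Shield damage boost, that adds firing FX and custom damage numbers to the owning player pawn.
; Tag wait tasks watch for the firing tag, binding to the current weapon's OnConsumeAmmo when it appears and unbinding when it clears.
; Each ammo event fires a gameplay cue if the weapon's tags match.
; Firing also adds an instigator-based damage numbers component if absent, removed on ability end if the current weapon is valid.
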(event Ability|EventFailedToActivatePassiveAbility
  (params (ActorInfo :type Gameplay%20Ability%20Actor%20Info%20Structure))
  (Development|PrintString "CombatEfficiencyVFX: Activation FAILED" true true "(R=1.000000,G=0.000000,B=0.332315,A=1.000000)" 2e+01))

(event Ability|EventActivateAbility
  (bind _returnvalue (CallFunction|GetAvatarasFortPlayerPawn))
  (bind _firingtag (Variables|Default|GetFiringTag))
  ;[1 1 1] EVENT -> Activate Ability
  (|CallParentFunction)
  (Utilities|IsValid _returnvalue
    (:Is%20Valid
      (bind _output_get (Variables|Default|SetOwningPawn _returnvalue))
      (bind _asynctaskproxy (Ability|Tasks|WaitGameplayTagAddToActor _output_get _firingtag)
        (:then
          (Variables|Default|SetAsyncWaitAttackingStart _asynctaskproxy)
          (bind _asynctaskproxy_2 (Ability|Tasks|WaitGameplayTagRemoveFromActor _output_get _firingtag)
            (:then
              (Variables|Default|SetAsyncWaitAttackingEnd _asynctaskproxy_2))
            (:Removed
              (bind _owningpawn (Variables|Default|GetOwningPawn)
                (:then
                  (Utilities|IsValid (Variables|/Script/FortniteGame.FortPawn|GetCurrentWeapon _owningpawn)
                    (:Is%20Valid
                      (EventDispatchers|UnbindEventfromOnConsumeAmmo (Variables|/Script/FortniteGame.FortPawn|GetCurrentWeapon _owningpawn) (EventDispatchers|CreateEvent :!event Custom|AmmoConsumed)))))
                (:else)))))
        (:Added
          (bind _owningpawn_2 (Variables|Default|GetOwningPawn)
            (:then
              (Utilities|IsValid (Variables|/Script/FortniteGame.FortPawn|GetCurrentWeapon _owningpawn_2)
                (:Is%20Valid
                  (EventDispatchers|BindEventtoOnConsumeAmmo (Variables|/Script/FortniteGame.FortPawn|GetCurrentWeapon _owningpawn_2) (EventDispatchers|CreateEvent :!event Custom|AmmoConsumed))
                  ;;[1 0.9047 0 0.502] Attach a Custom Damage Number component for Current actor and set it to be used as Instigator
                  (Utilities|IsValid (Actor|GetComponentbyClass _owningpawn_2 "/Script/DamageNumbersUI.CustomDamageNumbersDataComponent")
                    (:Is%20Not%20Valid
                      (bind _returnvalue_2 (Game|AddComponentbyClass _owningpawn_2 "/Script/DamageNumbersUI.CustomDamageNumbersDataComponent"))
                      (Class|CustomDamageNumbersDataComponent|SetDamageNumbersClass _returnvalue_2 (Variables|Default|GetNew%20Damage%20Numbers%20Class))
                      (Class|CustomDamageNumbersDataComponent|SetUsingInstigatorData _returnvalue_2 true)
                      (Fortnite|Utilities|RegisterComponent _returnvalue_2))))))
            (:else)))))))

(event Custom|AmmoConsumed
  (params (Quantity :type Integer))
  (CallFunction|ExecuteGameplayCueonOwningPawn))

(event Ability|EventOnEndAbility
  (params (bWasCancelled :type Boolean))
  (Utilities|FlowControl|Sequence
    (:then_0
      (bind _owningpawn (Variables|Default|GetOwningPawn)
        (:then
          (Utilities|IsValid (Variables|/Script/FortniteGame.FortPawn|GetCurrentWeapon _owningpawn)
            (:Is%20Valid
              (EventDispatchers|UnbindEventfromOnConsumeAmmo (Variables|/Script/FortniteGame.FortPawn|GetCurrentWeapon _owningpawn) (EventDispatchers|CreateEvent :!event Custom|AmmoConsumed))
              ;[1 0.9041 0 0.5] CleanUp Player Custom Damage numbers component only if using it as Instigator
              (Utilities|IsValid (Actor|GetComponentbyClass _owningpawn "/Script/DamageNumbersUI.CustomDamageNumbersDataComponent")
                (:Is%20Valid
                  (bind _returnvalue (Class|CustomDamageNumbersDataComponent|IsUsingInstigatorData (Actor|GetComponentbyClass _owningpawn "/Script/DamageNumbersUI.CustomDamageNumbersDataComponent")))
                  (if _returnvalue
                    (Components|RemoveArbitraryComponent _owningpawn "/Script/DamageNumbersUI.CustomDamageNumbersDataComponent")))))))
        (:else)))
    (:then_1
      (Ability|Async|EndAction (Variables|Default|GetAsyncWaitAttackingStart))
      (Ability|Async|EndAction (Variables|Default|GetAsyncWaitAttackingEnd)))
    (:then_2
      (|CallParentFunction bWasCancelled))))

(fn ExecuteGameplayCueOnOwningPawn
  (bind _owningpawn (Variables|Default|GetOwningPawn)
    (:then
      (if (Tags|WeaponDataHasAnyMatchingGameplayTags (Variables|/Script/FortniteGame.FortPawn|GetCurrentWeapon _owningpawn) (Variables|Default|GetValidWeaponTags))
        (Ability|ExecuteGameplayCueOnOwner (Variables|Default|GetGCTag_FiringFX) (Class|AthenaFunctionLibrary|MakeSimpleFortEffectContext _owningpawn (Variables|/Script/FortniteGame.FortPawn|GetCurrentWeapon _owningpawn)))))
    (:else)))
